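# Translucent, unlit, two-sided beam and ribbon trail material for the Street Racer Drift Remix glider FX.
# A panning beam texture times a scrolling gradient and soft edge mask drives colored emissive and opacity, with depth fade outside mobile.
# Particle color channels act as controls: R scales brightness, G offsets gradient timing, B sets edge hardness.
# World position offset adds scrolling, actor-relative vector noise to optional SplineThicken width.
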Material /Game/Gadgets/Assets/VinderTech_GliderChute/Glider_Street_Racer_Drift_Remix/FX/M_Glider_StreetRacer_DriftRemix_Beam
graph {
 properties {
  bUsedWithBeamTrails=true
  bUsedWithNiagaraRibbons=true
  blendMode=BLEND_Translucent
  pixelDepthOffsetMode=PDOM_Legacy
  shadingModel=MSM_Unlit
  twoSided=true
  usedShadingModels=Unlit
 }
 nodes [
  n1 = WorldPosition()
  n2 = ActorPositionWS()
  n3 = ScalarParameter() {defaultValue=12, parameterName=CurlSpeed}
  n4 = Time()
  n5 = Divide(A=Subtract(A=Subtract(A=@n1.XYZ, B=@n2), B=Multiply(A=@n3, B=@n4)), B=ScalarParameter() {defaultValue=25, parameterName=CurlScale})
  n6 = ScalarParameter() {defaultValue=20, parameterName=SplineWidthBase}
  n7 = ScalarParameter() {defaultValue=20, parameterName=SplineWidthTip}
  n8 = StaticBool() {value=true}
  Add(
   A=Multiply(A=ScalarParameter() {defaultValue=6, parameterName=CurlIntensity}, B=VectorNoise(World Position=@n5) {noiseFunction=VNF_VectorALU})
   B=StaticSwitchParameter(
    True=MaterialFunctionCall(WidthBase=@n6, WidthTip=@n7, Expand U or V UV Channel=@n8) {materialFunction=&SplineThicken}.WorldPositionOffset
    False=Constant()
   ) {parameterName=UseSplineMesh}
  )
  n9 = VectorParameter() {parameterName=Color, defaultValue=<rgba 20 0 8.0759 0>}
  n10 = TextureCoordinate()
  n11 = LinearInterpolate(Alpha=ParticleColor().B) {constA=0.4, constB=0.8}
  n12 = Add(
   A=SphereMask(A=@n10, B=Constant2Vector() {g=1, r=0.5}, Hardness=@n11) {attenuationRadius=0.5, hardnessPercent=0}
   B=SphereMask(A=@n10, B=Constant2Vector() {r=0.5}, Hardness=@n11) {attenuationRadius=0.5, hardnessPercent=0}
  )
  n13 = TextureCoordinate()
  n14 = MaterialFunctionCall(XY=@n13) {materialFunction=&Swizzle}
  n15 = TextureSample(UVs=Panner(Coordinate=StaticSwitchParameter(True=@n14.YX, False=@n13) {parameterName=SwizzleTexture}) {speedX=1}) {samplerType=SAMPLERTYPE_LinearColor, texture=/Game/Effects/Fort_Effects/Textures/Tile/T_AMR_Beam_02}
  n16 = Add(A=@n15.R, B=@n15.R) {constB=0}
  n17 = TextureCoordinate() {uTiling=0.03}
  n18 = MaterialFunctionCall(XY=@n17) {materialFunction=&Swizzle}
  n19 = Time()
  n20 = ParticleColor()
  n21 = TextureSample(UVs=Panner(Coordinate=StaticSwitchParameter(True=@n18.YX, False=@n17) {defaultValue=true, parameterName=SwizzleMask}, Time=Add(A=@n19, B=@n20.G)) {speedY=0.3}) {texture=/Game/Global/Textures/Default/Masks_and_Gradients/T_Vert_Gradient_Sections}
  n22 = Add(A=Multiply(A=OneMinus(None=@n12), B=Multiply(A=@n16, B=@n21.R))) {constB=0}
  MaterialFunctionCall(Emissive=Multiply(A=ParticleColor().R, B=Multiply(A=@n9.RGB, B=@n22))) {materialFunction=&MF_FX_EmissiveAdjustments}
  ShadingPathSwitch(Default=DepthFade(Opacity=@n22) {fadeDistanceDefault=20}, Mobile=@n22) {desc=Migrated from FeatureLevelSwitch}
 ]
}
